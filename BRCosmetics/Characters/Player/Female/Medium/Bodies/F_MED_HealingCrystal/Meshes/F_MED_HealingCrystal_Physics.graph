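# Physics asset for the F_MED_HealingCrystal female body cosmetic, covering the pelvis, thighs and six dynamic bone chains around the hips.
# Front, rear and side chains on each side run from a tiny master capsule through three larger capsules.
# Master bodies rotate freely against the pelvis, while the links below them are angularly locked.
# Pelvis and thigh capsules form the leg volume, with limit angles set on the thigh joints.
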PhysicsAsset /BRCosmetics/Characters/Player/Female/Medium/Bodies/F_MED_HealingCrystal/Meshes/F_MED_HealingCrystal_Physics
bodies {
 dyn_Front_01_l [{shapeType=Capsule, length=20, radius=6, shapeName=dyn_Front_01_l_capsule, center=<6 0 12>, rotation=<rot 0 0 -110>}]
 dyn_Front_01_r [{shapeType=Capsule, length=20, radius=6, shapeName=dyn_Front_01_r_capsule, center=<6 0 10>, rotation=<rot 0 0 -70>}]
 dyn_Front_02_l [{shapeType=Capsule, length=20, radius=6, shapeName=dyn_Front_02_l_capsule, center=<6 0 12>, rotation=<rot 0 0 -110>}]
 dyn_Front_02_r [{shapeType=Capsule, length=20, radius=6, shapeName=dyn_Front_02_r_capsule, center=<6 0 10>, rotation=<rot 0 0 -70>}]
 dyn_Front_03_l [{shapeType=Capsule, length=20, radius=6, shapeName=dyn_Front_03_l_capsule, center=<6 0 12>, rotation=<rot 0 0 -110>}]
 dyn_Front_03_r [{shapeType=Capsule, length=20, radius=6, shapeName=dyn_Front_03_r_capsule, center=<6 0 10>, rotation=<rot 0 0 -70>}]
 dyn_Rear_01_l [{shapeType=Capsule, length=20, radius=7.8, shapeName=dyn_Rear_01_l_capsule, center=<6 0 6>, rotation=<rot 0 0 -70>}]
 dyn_Rear_01_r [{shapeType=Capsule, length=20, radius=7.6, shapeName=dyn_Rear_01_r_capsule, center=<6 0 6>, rotation=<rot 0 0 -110>}]
 dyn_Rear_02_l [{shapeType=Capsule, length=20, radius=7.8, shapeName=dyn_Rear_02_l_capsule, center=<6 0 6>, rotation=<rot 0 0 -70>}]
 dyn_Rear_02_r [{shapeType=Capsule, length=20, radius=7.6, shapeName=dyn_Rear_02_r_capsule, center=<6 0 6>, rotation=<rot 0 0 -110>}]
 dyn_Rear_03_l [{shapeType=Capsule, length=20, radius=8, shapeName=dyn_Rear_03_l_capsule, center=<6 0 6>, rotation=<rot 0 0 -70>}]
 dyn_Rear_03_r [{shapeType=Capsule, length=20, radius=7.8, shapeName=dyn_Rear_03_r_capsule, center=<6 0 6>, rotation=<rot 0 0 -110>}]
 dyn_Side_01_l [{shapeType=Capsule, length=20, radius=7, shapeName=dyn_Side_01_l_capsule, center=<6 0 9.2>, rotation=<rot 0 0 -90>}]
 dyn_Side_01_r [{shapeType=Capsule, length=20, radius=8, shapeName=dyn_Side_01_r_capsule, center=<6 0 9.5>, rotation=<rot 0 0 -90>}]
 dyn_Side_02_l [{shapeType=Capsule, length=20, radius=7, shapeName=dyn_Side_02_l_capsule, center=<9.8276 0 7.227>, rotation=<rot 0 0 -90>}]
 dyn_Side_02_r [{shapeType=Capsule, length=20, radius=8, shapeName=dyn_Side_02_r_capsule, center=<6 0 9.5>, rotation=<rot 0 0 -90>}]
 dyn_Side_03_l [{shapeType=Capsule, length=20, radius=7, shapeName=dyn_Side_03_l_capsule, center=<9.2504 0 6.0298>, rotation=<rot 0 0 -90>}]
 dyn_Side_03_r [{shapeType=Capsule, length=20, radius=8, shapeName=dyn_Side_03_r_capsule, center=<6 0 10>, rotation=<rot 0 0 -90>}]
 mstr_Front_l [{shapeType=Capsule, length=0.505, radius=0.505, shapeName=mstr_Front_l_capsule, rotation=<rot 0 -90 0>}]
 mstr_Front_r [{shapeType=Capsule, length=0.505, radius=0.505, shapeName=mstr_Front_r_capsule, rotation=<rot 0 -90 0>}]
 mstr_Rear_l [{shapeType=Capsule, length=0.505, radius=0.505, shapeName=mstr_Rear_l_capsule, rotation=<rot 0 -90 0>}]
 mstr_Rear_r [{shapeType=Capsule, length=0.505, radius=0.505, shapeName=mstr_Rear_r_capsule, rotation=<rot 0 -90 0>}]
 mstr_Side_l [{shapeType=Capsule, length=0.505, radius=0.505, shapeName=mstr_Side_l_capsule, rotation=<rot 0 -90 0>}]
 mstr_Side_r [{shapeType=Capsule, length=0.505, radius=0.505, shapeName=mstr_Side_r_capsule, rotation=<rot 0 -90 0>}]
 pelvis [{shapeType=Capsule, length=10, radius=10}]
 thigh_l [{shapeType=Capsule, length=30, radius=9.4871, center=<-20 0 0>, rotation=<rot -90 0 0>}]
 thigh_r [{shapeType=Capsule, length=30, radius=9.479, center=<20 0 0>, rotation=<rot -90 0 0>}]
}
constraints [
 {bone1Name=thigh_l, bone2Name=pelvis, swing1LimitDegrees=45, swing2LimitDegrees=45, twistLimitDegrees=45}
 {bone1Name=thigh_r, bone2Name=pelvis, swing1LimitDegrees=45, swing2LimitDegrees=45, twistLimitDegrees=45}
 {
  bone1Name=mstr_Front_l
  bone2Name=pelvis
  swing1LimitDegrees=45
  swing1Motion=Free
  swing2LimitDegrees=45
  swing2Motion=Free
  twistLimitDegrees=45
  twistMotion=Free
 }
 {
  bone1Name=dyn_Front_01_l
  bone2Name=mstr_Front_l
  swing1LimitDegrees=45
  swing1Motion=Locked
  swing2LimitDegrees=45
  swing2Motion=Locked
  twistLimitDegrees=45
  twistMotion=Locked
 }
 {
  bone1Name=dyn_Front_02_l
  bone2Name=dyn_Front_01_l
  swing1LimitDegrees=45
  swing1Motion=Locked
  swing2LimitDegrees=45
  swing2Motion=Locked
  twistLimitDegrees=45
  twistMotion=Locked
 }
 {
  bone1Name=dyn_Front_03_l
  bone2Name=dyn_Front_02_l
  swing1LimitDegrees=45
  swing1Motion=Locked
  swing2LimitDegrees=45
  swing2Motion=Locked
  twistLimitDegrees=45
  twistMotion=Locked
 }
 {
  bone1Name=mstr_Front_r
  bone2Name=pelvis
  swing1LimitDegrees=45
  swing1Motion=Free
  swing2LimitDegrees=45
  swing2Motion=Free
  twistLimitDegrees=45
  twistMotion=Free
 }
 {
  bone1Name=dyn_Front_01_r
  bone2Name=mstr_Front_r
  swing1LimitDegrees=45
  swing1Motion=Locked
  swing2LimitDegrees=45
  swing2Motion=Locked
  twistLimitDegrees=45
  twistMotion=Locked
 }
 {
  bone1Name=dyn_Front_02_r
  bone2Name=dyn_Front_01_r
  swing1LimitDegrees=45
  swing1Motion=Locked
  swing2LimitDegrees=45
  swing2Motion=Locked
  twistLimitDegrees=45
  twistMotion=Locked
 }
 {
  bone1Name=dyn_Front_03_r
  bone2Name=dyn_Front_02_r
  swing1LimitDegrees=45
  swing1Motion=Locked
  swing2LimitDegrees=45
  swing2Motion=Locked
  twistLimitDegrees=45
  twistMotion=Locked
 }
 {
  bone1Name=mstr_Rear_l
  bone2Name=pelvis
  swing1LimitDegrees=45
  swing1Motion=Free
  swing2LimitDegrees=45
  swing2Motion=Free
  twistLimitDegrees=45
  twistMotion=Free
 }
 {
  bone1Name=dyn_Rear_01_l
  bone2Name=mstr_Rear_l
  swing1LimitDegrees=45
  swing1Motion=Locked
  swing2LimitDegrees=45
  swing2Motion=Locked
  twistLimitDegrees=45
  twistMotion=Locked
 }
 {
  bone1Name=dyn_Rear_02_l
  bone2Name=dyn_Rear_01_l
  swing1LimitDegrees=45
  swing1Motion=Locked
  swing2LimitDegrees=45
  swing2Motion=Locked
  twistLimitDegrees=45
  twistMotion=Locked
 }
 {
  bone1Name=dyn_Rear_03_l
  bone2Name=dyn_Rear_02_l
  swing1LimitDegrees=45
  swing1Motion=Locked
  swing2LimitDegrees=45
  swing2Motion=Locked
  twistLimitDegrees=45
  twistMotion=Locked
 }
 {
  bone1Name=mstr_Rear_r
  bone2Name=pelvis
  swing1LimitDegrees=45
  swing1Motion=Free
  swing2LimitDegrees=45
  swing2Motion=Free
  twistLimitDegrees=45
  twistMotion=Free
 }
 {
  bone1Name=dyn_Rear_01_r
  bone2Name=mstr_Rear_r
  swing1LimitDegrees=45
  swing1Motion=Locked
  swing2LimitDegrees=45
  swing2Motion=Locked
  twistLimitDegrees=45
  twistMotion=Locked
 }
 {
  bone1Name=dyn_Rear_02_r
  bone2Name=dyn_Rear_01_r
  swing1LimitDegrees=45
  swing1Motion=Locked
  swing2LimitDegrees=45
  swing2Motion=Locked
  twistLimitDegrees=45
  twistMotion=Locked
 }
 {
  bone1Name=dyn_Rear_03_r
  bone2Name=dyn_Rear_02_r
  swing1LimitDegrees=45
  swing1Motion=Locked
  swing2LimitDegrees=45
  swing2Motion=Locked
  twistLimitDegrees=45
  twistMotion=Locked
 }
 {
  bone1Name=mstr_Side_l
  bone2Name=pelvis
  swing1LimitDegrees=45
  swing1Motion=Free
  swing2LimitDegrees=45
  swing2Motion=Free
  twistLimitDegrees=45
  twistMotion=Free
 }
 {
  bone1Name=dyn_Side_01_l
  bone2Name=mstr_Side_l
  swing1LimitDegrees=45
  swing1Motion=Locked
  swing2LimitDegrees=45
  swing2Motion=Locked
  twistLimitDegrees=45
  twistMotion=Locked
 }
 {
  bone1Name=dyn_Side_02_l
  bone2Name=dyn_Side_01_l
  swing1LimitDegrees=45
  swing1Motion=Locked
  swing2LimitDegrees=45
  swing2Motion=Locked
  twistLimitDegrees=45
  twistMotion=Locked
 }
 {
  bone1Name=dyn_Side_03_l
  bone2Name=dyn_Side_02_l
  swing1LimitDegrees=45
  swing1Motion=Locked
  swing2LimitDegrees=45
  swing2Motion=Locked
  twistLimitDegrees=45
  twistMotion=Locked
 }
 {
  bone1Name=mstr_Side_r
  bone2Name=pelvis
  swing1LimitDegrees=45
  swing1Motion=Free
  swing2LimitDegrees=45
  swing2Motion=Free
  twistLimitDegrees=45
  twistMotion=Free
 }
 {
  bone1Name=dyn_Side_01_r
  bone2Name=mstr_Side_r
  swing1LimitDegrees=45
  swing1Motion=Locked
  swing2LimitDegrees=45
  swing2Motion=Locked
  twistLimitDegrees=45
  twistMotion=Locked
 }
 {
  bone1Name=dyn_Side_02_r
  bone2Name=dyn_Side_01_r
  swing1LimitDegrees=45
  swing1Motion=Locked
  swing2LimitDegrees=45
  swing2Motion=Locked
  twistLimitDegrees=45
  twistMotion=Locked
 }
 {
  bone1Name=dyn_Side_03_r
  bone2Name=dyn_Side_02_r
  swing1LimitDegrees=45
  swing1Motion=Locked
  swing2LimitDegrees=45
  swing2Motion=Locked
  twistLimitDegrees=45
  twistMotion=Locked
 }
]
