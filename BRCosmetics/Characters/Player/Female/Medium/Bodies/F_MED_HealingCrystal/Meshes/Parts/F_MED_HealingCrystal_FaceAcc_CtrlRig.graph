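# Control Rig for the HealingCrystal female face accessory mesh that poses hair bones procedurally as the body moves.
# Spherical pose readers on head, neck_02 and neck_01 drive weights for Modify Transforms on top, middle, lower and side hair chains.
# Pelvis height, remapped through a curve that peaks midway, sets CrouchWalk to weight an extra hair offset.
# When PistolEquip is true, another head-versus-neck reader adds a further hair correction.
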ControlRigBlueprint /BRCosmetics/Characters/Player/Female/Medium/Bodies/F_MED_HealingCrystal/Meshes/Parts/F_MED_HealingCrystal_FaceAcc_CtrlRig
graphs [
 RigVMFunctionLibrary {links [], nodes []}
 RigVMFunction_Sequence_1_ContainedGraph {nodes [
   n1 = RigVMFunctionEntryNode()
   RigVMFunction_Sequence = &RigVMFunction_Sequence(ExecuteContext=@n1.ExecuteContext)
   n2 = &RigVMFunction_Sequence(ExecuteContext=@RigVMFunction_Sequence.B)
   n3 = &RigVMFunction_Sequence(ExecuteContext=@n2.B)
   n4 = &RigVMFunction_Sequence(ExecuteContext=@n3.B)
   n5 = &RigVMFunction_Sequence(ExecuteContext=@n4.B)
   RigVMFunctionReturnNode(A=@RigVMFunction_Sequence.A, B=@n2.A, C=@n3.A, D=@n4.A, E=@n5.A, F=@n5.B)
  ]}
 RigVMModel {nodes [
   RigUnit_BeginExecution = &RigUnit_BeginExecution()
   RigVMFunction_Sequence = RigVMAggregateNode(ExecuteContext=@RigUnit_BeginExecution.ExecutePin)
   &RigUnit_ModifyTransforms(
    ItemToModify=((Item=(Type=Bone,Name="neck_02"),Transform=(Rotation=(X=0,Y=0,Z=0,W=1),Translation=(X=0,Y=0,Z=0),Scale3D=(X=1,Y=1,Z=1))))
    Weight=1
    WeightMinimum=0
    ExecutePin=@RigVMFunction_Sequence.A
   )
   > n1 = &RigUnit_SphericalPoseReader(
    ActiveRegionScaleFactors=(PositiveWidth=1,NegativeWidth=1,PositiveHeight=1,NegativeHeight=1)
    ActiveRegionSize=0.05
    Debug=(bDrawDebug=false,bDraw2D=False,bDrawLocalAxes=False,DebugScale=25,DebugSegments=20,DebugThickness=0.25)
    DriverAxis=(X=1,Y=0,Z=0)
    DriverItem=(Type=Bone,Name="head")
    FalloffRegionScaleFactors=(PositiveWidth=1,NegativeWidth=1,PositiveHeight=1,NegativeHeight=1)
    FalloffSize=0.5
    FlipHeightScaling=False
    FlipWidthScaling=False
    OptionalParentItem=(Type=Bone,Name="None")
    RotationOffset=(X=-90,Y=0,Z=0)
   )
   > &RigUnit_ModifyTransforms(
    ItemToModify="((Item=(Type=Bone,Name=\"dyn_mid_hair_01\"),Transform=(Rotation=(X=0,Y=-0.342,Z=0,W=0.9397),Translation=(X=0,Y=0,Z=3),Scale3D=(X=0.8,Y=1,Z=1))),(Item=(Type=Bone,Name=\"dyn_mid_hair_02\"),Transform=(Rotation=(X=0,Y=-0.342,Z=0,W=0.9397),Translation=(X=0,Y=0,Z=3) ...(truncated)"
    WeightMinimum=0
    Weight=@n1.OutputParam
   )
   > &RigUnit_ModifyTransforms(
    ItemToModify="((Item=(Type=Bone,Name=\"dyn_lower_hair_01\"),Transform=(Rotation=(X=0,Y=-0.342,Z=0,W=0.9397),Translation=(X=0,Y=0,Z=0),Scale3D=(X=0.8,Y=1,Z=1))),(Item=(Type=Bone,Name=\"dyn_lower_hair_02\"),Transform=(Rotation=(X=0,Y=-0.1736,Z=0,W=0.9848),Translation=(X=0,Y=0 ...(truncated)"
    WeightMinimum=0
    Weight=@n1.OutputParam
   )
   > &RigUnit_ModifyTransforms(
    ItemToModify="((Item=(Type=Bone,Name=\"dyn_top_hair_01\"),Transform=(Rotation=(X=0,Y=-0.0872,Z=0,W=0.9962),Translation=(X=0,Y=0,Z=0),Scale3D=(X=1,Y=1,Z=1))),(Item=(Type=Bone,Name=\"dyn_top_hair_02\"),Transform=(Rotation=(X=0,Y=-0.1736,Z=0,W=0.9848),Translation=(X=0,Y=0,Z=0) ...(truncated)"
    WeightMinimum=0
    Weight=@n1.OutputParam
   )
   > n2 = &RigUnit_SphericalPoseReader(
    ActiveRegionScaleFactors=(PositiveWidth=1,NegativeWidth=1,PositiveHeight=1,NegativeHeight=1)
    ActiveRegionSize=0.05
    Debug=(bDrawDebug=false,bDraw2D=False,bDrawLocalAxes=False,DebugScale=25,DebugSegments=20,DebugThickness=0.25)
    DriverAxis=(X=1,Y=0,Z=0)
    DriverItem=(Type=Bone,Name="neck_02")
    FalloffRegionScaleFactors=(PositiveWidth=1,NegativeWidth=1,PositiveHeight=1,NegativeHeight=1)
    FalloffSize=0.5
    FlipHeightScaling=False
    FlipWidthScaling=False
    OptionalParentItem=(Type=Bone,Name="None")
    RotationOffset=(X=-90,Y=0,Z=0)
   )
   > &RigUnit_ModifyTransforms(
    ItemToModify="((Item=(Type=Bone,Name=\"dyn_mid_hair_01\"),Transform=(Rotation=(X=0,Y=-0.2588,Z=0,W=0.9659),Translation=(X=0,Y=0,Z=3),Scale3D=(X=1,Y=1,Z=1))),(Item=(Type=Bone,Name=\"dyn_mid_hair_02\"),Transform=(Rotation=(X=0,Y=-0.2588,Z=0,W=0.9659),Translation=(X=0,Y=0,Z=3) ...(truncated)"
    WeightMinimum=0
    Weight=@n2.OutputParam
   )
   > &RigUnit_ModifyTransforms(
    ItemToModify="((Item=(Type=Bone,Name=\"dyn_lower_hair_01\"),Transform=(Rotation=(X=0,Y=-0.1736,Z=0,W=0.9848),Translation=(X=0,Y=0,Z=0),Scale3D=(X=0.5,Y=1,Z=1))),(Item=(Type=Bone,Name=\"dyn_lower_hair_02\"),Transform=(Rotation=(X=0,Y=-0.1736,Z=0,W=0.9848),Translation=(X=0,Y= ...(truncated)"
    WeightMinimum=0
    Weight=@n2.OutputParam
   )
   &RigUnit_ModifyTransforms(
    ItemToModify=((Item=(Type=Bone,Name="neck_01"),Transform=(Rotation=(X=0,Y=0,Z=0,W=1),Translation=(X=0,Y=0,Z=0),Scale3D=(X=1,Y=1,Z=1))))
    Weight=1
    WeightMinimum=0
    ExecutePin=@RigVMFunction_Sequence.B
   )
   > n3 = &RigUnit_SphericalPoseReader(
    ActiveRegionScaleFactors=(PositiveWidth=1,NegativeWidth=1,PositiveHeight=1,NegativeHeight=1)
    ActiveRegionSize=0.05
    Debug=(bDrawDebug=true,bDraw2D=False,bDrawLocalAxes=False,DebugScale=25,DebugSegments=20,DebugThickness=0.25)
    DriverAxis=(X=1,Y=0,Z=0)
    DriverItem=(Type=Bone,Name="neck_01")
    FalloffRegionScaleFactors=(PositiveWidth=0.77,NegativeWidth=0.1,PositiveHeight=0.2,NegativeHeight=0.4)
    FalloffSize=0.5
    FlipHeightScaling=False
    FlipWidthScaling=False
    OptionalParentItem=(Type=Bone,Name="spine_05")
    RotationOffset=(X=-60,Y=0,Z=18)
   )
   > &RigUnit_ModifyTransforms(
    ItemToModify="((Item=(Type=Bone,Name=\"dyn_mid_hair_01_r\"),Transform=(Rotation=(X=-0.0449,Y=-0.2549,Z=-0.1677,W=0.9513),Translation=(X=0,Y=0,Z=0),Scale3D=(X=1,Y=1,Z=1))),(Item=(Type=Bone,Name=\"dyn_mid_hair_02_r\"),Transform=(Rotation=(X=-0.0226,Y=-0.2578,Z=-0.0842,W=0.962 ...(truncated)"
    WeightMinimum=0
    Weight=@n3.OutputParam
   )
   > &RigUnit_ModifyTransforms(
    ItemToModify="((Item=(Type=Bone,Name=\"dyn_lower_hair_01_r\"),Transform=(Rotation=(X=0,Y=0,Z=-0.342,W=0.9397),Translation=(X=0,Y=0,Z=0),Scale3D=(X=1,Y=1,Z=1))),(Item=(Type=Bone,Name=\"dyn_lower_hair_02_r\"),Transform=(Rotation=(X=0,Y=0,Z=0,W=1),Translation=(X=0,Y=0,Z=0),Sca ...(truncated)"
    WeightMinimum=0
    Weight=@n3.OutputParam
   )
   > &RigUnit_ModifyTransforms(
    ItemToModify="((Item=(Type=Bone,Name=\"mstr_lower_hair_01_l\"),Transform=(Rotation=(X=0,Y=0.0872,Z=0,W=0.9962),Translation=(X=0,Y=0,Z=0),Scale3D=(X=1,Y=1,Z=1))),(Item=(Type=Bone,Name=\"mstr_lower_hair_02_l\"),Transform=(Rotation=(X=0,Y=0.0872,Z=0,W=0.9962),Translation=(X=-0 ...(truncated)"
    WeightMinimum=0
    Weight=@n3.OutputParam
   )
   > &RigUnit_ModifyTransforms(
    ItemToModify="((Item=(Type=Bone,Name=\"dyn_mid_hail_01_l\"),Transform=(Rotation=(X=0,Y=0,Z=-0.342,W=0.9397),Translation=(X=0,Y=0,Z=0),Scale3D=(X=1,Y=1,Z=1))),(Item=(Type=Bone,Name=\"dyn_mid_hail_02_l\"),Transform=(Rotation=(X=0,Y=0,Z=0.1736,W=0.9848),Translation=(X=0,Y=0,Z= ...(truncated)"
    WeightMinimum=0
    Weight=@n3.OutputParam
   )
   > n4 = &RigUnit_SphericalPoseReader(
    ActiveRegionScaleFactors=(PositiveWidth=1,NegativeWidth=1,PositiveHeight=1,NegativeHeight=1)
    ActiveRegionSize=0.05
    Debug=(bDrawDebug=true,bDraw2D=False,bDrawLocalAxes=False,DebugScale=25,DebugSegments=20,DebugThickness=0.25)
    DriverAxis=(X=1,Y=0,Z=0)
    DriverItem=(Type=Bone,Name="neck_01")
    FalloffRegionScaleFactors=(PositiveWidth=0.65,NegativeWidth=0.8,PositiveHeight=0.4,NegativeHeight=0.5)
    FalloffSize=0.5
    FlipHeightScaling=False
    FlipWidthScaling=False
    OptionalParentItem=(Type=Bone,Name="spine_05")
    RotationOffset=(X=10,Y=-30,Z=0)
   )
   > &RigUnit_ModifyTransforms(
    ItemToModify="((Item=(Type=Bone,Name=\"dyn_mid_hair_01_r\"),Transform=(Rotation=(X=-0.0931,Y=-0.2486,Z=-0.3584,W=0.895),Translation=(X=0,Y=0,Z=0),Scale3D=(X=1,Y=1,Z=1))),(Item=(Type=Bone,Name=\"dyn_mid_hair_02_r\"),Transform=(Rotation=(X=0.0076,Y=0.0868,Z=-0.0868,W=0.9924), ...(truncated)"
    WeightMinimum=0
    Weight=@n4.OutputParam
   )
   > &RigUnit_ModifyTransforms(
    ItemToModify="((Item=(Type=Bone,Name=\"dyn_mid_hail_01_l\"),Transform=(Rotation=(X=0,Y=0,Z=-0.5736,W=0.8192),Translation=(X=0,Y=0,Z=0),Scale3D=(X=1,Y=1,Z=1))),(Item=(Type=Bone,Name=\"dyn_mid_hail_02_l\"),Transform=(Rotation=(X=0,Y=0,Z=-0.2588,W=0.9659),Translation=(X=0,Y=0, ...(truncated)"
    WeightMinimum=0
    Weight=@n4.OutputParam
   )
   n5 = &RigVMFunction_Sequence(ExecuteContext=@RigVMFunction_Sequence.C)
   n6 = &RigVMFunction_AnimEvalRichCurve(
    Curve=(EditorCurveData=(Keys=((),(Time=0.5,Value=1),(Time=0.8))))
    SourceMaximum=2
    SourceMinimum=0
    TargetMinimum=0
    Value=&RigVMFunction_MathDoubleRemap(
     SourceMaximum=75
     SourceMinimum=45
     TargetMaximum=1
     TargetMinimum=0
     bClamp=false
     Value=&RigUnit_GetRelativeTransformForItem(Child=(Type=Bone,Name="pelvis"), Parent=(Type=Bone,Name="root"), bChildInitial=False, bParentInitial=False).RelativeTransform.Translation.Z
    ).Result
   )
   RigVMVariableNode(Value=@n6.Result, ExecuteContext=@n5.A) {variable=CrouchWalk}
   &RigUnit_ModifyTransforms(
    ItemToModify="((Item=(Type=Bone,Name=\"dyn_mid_hair_01_r\"),Transform=(Rotation=(X=0,Y=-0.1736,Z=0,W=0.9848),Translation=(X=0,Y=0,Z=0),Scale3D=(X=1,Y=1,Z=1))),(Item=(Type=Bone,Name=\"mstr_lower_hair_02_l\"),Transform=(Rotation=(X=0,Y=0.1736,Z=0,W=0.9848),Translation=(X=0,Y= ...(truncated)"
    WeightMinimum=0
    ExecutePin=@n5.B
    Weight=RigVMVariableNode() {variable=CrouchWalk}.Value
   )
   n7 = &RigVMFunction_Sequence(ExecuteContext=@RigVMFunction_Sequence.D)
   RigVMFunction_ControlFlowBranch = &RigVMFunction_ControlFlowBranch(ExecuteContext=@n7.A, Condition=RigVMVariableNode() {variable=PistolEquip}.Value)
   n8 = &RigUnit_SphericalPoseReader(
    ActiveRegionScaleFactors=(PositiveWidth=1,NegativeWidth=1,PositiveHeight=1,NegativeHeight=1)
    ActiveRegionSize=0.1
    Debug=(bDrawDebug=true,DebugScale=25,DebugSegments=20,DebugThickness=0.25)
    DriverAxis=(X=0,Y=-1,Z=0)
    DriverItem=(Type=Bone,Name="head")
    FalloffRegionScaleFactors=(PositiveWidth=2,NegativeWidth=2,PositiveHeight=2,NegativeHeight=2)
    FalloffSize=0.2
    FlipWidthScaling=false
    OptionalParentItem=(Type=Bone,Name="neck_01")
    RotationOffset=(X=-90,Y=0,Z=-55)
    ExecutePin=@RigVMFunction_ControlFlowBranch.True
   )
   > &RigUnit_ModifyTransforms(
    ItemToModify="((Item=(Type=Bone,Name=\"dyn_lower_hair_01_r\"),Transform=(Rotation=(X=0,Y=0,Z=-0.1305,W=0.9914),Translation=(X=0,Y=0,Z=0),Scale3D=(X=1,Y=1,Z=1))),(Item=(Type=Bone,Name=\"dyn_lower_hair_02_r\"),Transform=(Rotation=(X=0.0076,Y=0.0868,Z=-0.0868,W=0.9924),Transla ...(truncated)"
    WeightMinimum=0
    Weight=@n8.OutputParam
   )
  ]}
]
hierarchy [
 root {loc=<0 0 0>, rot=<0 0 0 1>}
 pelvis {parent=root, loc=<0 -0.5301 98.0571>, rot=<0.0127 0.707 -0.0127 -0.707>}
 spine_01 {parent=pelvis, loc=<6.1208 -0.9328 0>, rot=<0 0 0.034 -0.9994>}
 spine_02 {parent=spine_01, loc=<6.5997 -0.2095 0>, rot=<0 0 0.0238 -0.9997>}
 spine_03 {parent=spine_02, loc=<6.4914 -0.1237 0>, rot=<0 0 0.0149 -0.9999>}
 spine_04 {parent=spine_03, loc=<6.9598 0.0794 0>, rot=<0 0 -0.0132 -0.9999>}
 spine_05 {parent=spine_04, loc=<6.9987 0 0>, rot=<0 0 -0.0015 -1>}
 neck_01 {parent=spine_05, loc=<11.3138 0.1529 0>, rot=<0 0 -0.2003 -0.9797>}
 neck_02 {parent=neck_01, loc=<4.5627 0 0>, rot=<0 0 0.0504 -0.9987>}
 head {parent=neck_02, loc=<5.3914 0 0>, rot=<0 0 0.115 -0.9934>}
 mstr_Hair {parent=head, loc=<9.0383 -1.7109 0>, rot=<0.2706 -0.6533 0.6533 -0.2706>}
 dyn_Hair {parent=mstr_Hair, loc=<0 0 0>, rot=<0 0 0 1>}
 mstr_top_hair {parent=dyn_Hair, loc=<-4.6317 -0.1111 2.6866>, rot=<0.0055 0.3283 -0.004 -0.9445>}
 dyn_top_hair_01 {parent=mstr_top_hair, loc=<0 0 0>, rot=<0 0 0 1>}
 dyn_top_hair_02 {parent=dyn_top_hair_01, loc=<6 0 0>, rot=<0 0 0 1>}
 dyn_top_hair_03 {parent=dyn_top_hair_02, loc=<6 0 0>, rot=<0 0 0 1>}
 mstr_mid_hair {parent=dyn_Hair, loc=<4.1939 0.8665 2.3685>, rot=<-0.0435 -0.0087 0.0048 -0.999>}
 dyn_mid_hair_01 {parent=mstr_mid_hair, loc=<0 0 0>, rot=<0 -0.0059 0 -1>}
 dyn_mid_hair_02 {parent=dyn_mid_hair_01, loc=<7 0 0>, rot=<0 -0.1174 0 -0.9931>}
 dyn_mid_hair_03 {parent=dyn_mid_hair_02, loc=<7 0 0>, rot=<0 -0.165 0 -0.9863>}
 mstr_lower_hair {parent=dyn_Hair, loc=<2.5931 0.8665 -1.5223>, rot=<-0.0438 -0.1454 -0.0012 -0.9884>}
 dyn_lower_hair_01 {parent=mstr_lower_hair, loc=<0 0 0>, rot=<0 -0.0059 0 -1>}
 dyn_lower_hair_02 {parent=dyn_lower_hair_01, loc=<7 0 0>, rot=<0 -0.1174 0 -0.9931>}
 dyn_lower_hair_03 {parent=dyn_lower_hair_02, loc=<7 0 0>, rot=<0 -0.165 0 -0.9863>}
 mstr_lower_hair_r {parent=mstr_lower_hair, loc=<1.672 -6.5659 0.7227>, rot=<0.076 -0.186 0.1581 -0.9668>}
 dyn_lower_hair_01_r {parent=mstr_lower_hair_r, loc=<0 0 0>, rot=<0 0 0 1>}
 dyn_lower_hair_02_r {parent=dyn_lower_hair_01_r, loc=<5 0 0>, rot=<0 -0.1174 0 -0.9931>}
 dyn_lower_hair_03_r {parent=dyn_lower_hair_02_r, loc=<5 0 0>, rot=<0 -0.165 0 -0.9863>}
 mstr_lower_hair_l {parent=mstr_lower_hair, loc=<1.8716 6.5389 -0.4108>, rot=<0.0048 -0.1978 -0.1268 -0.972>}
 mstr_lower_hair_01_l {parent=mstr_lower_hair_l, loc=<0 0 0>, rot=<0 0 0 1>}
 mstr_lower_hair_02_l {parent=mstr_lower_hair_01_l, loc=<5 0 0>, rot=<0 -0.1174 0 -0.9931>}
 mstr_lower_hair_03_l {parent=mstr_lower_hair_02_l, loc=<5 0 0>, rot=<0 -0.165 0 -0.9863>}
 mstl_mid_hail_l {parent=dyn_Hair, loc=<2.6921 6.3108 1.1818>, rot=<-0.0106 -0.1384 -0.2774 -0.9507>}
 dyn_mid_hail_01_l {parent=mstl_mid_hail_l, loc=<0 0 0>, rot=<0 0 0 1>}
 dyn_mid_hail_02_l {parent=dyn_mid_hail_01_l, loc=<6.0001 0 0>, rot=<0.0281 -0.084 0.0024 -0.9961>}
 dyn_mid_hail_03_l {parent=dyn_mid_hail_02_l, loc=<5.9998 0 0>, rot=<0 0 0 1>}
 mstr_mid_hair_r {parent=dyn_Hair, loc=<2.3872 -5.2052 1.2302>, rot=<-0.0032 -0.0664 0.3025 -0.9508>}
 dyn_mid_hair_01_r {parent=mstr_mid_hair_r, loc=<0 0 0>, rot=<0 0 0 1>}
 dyn_mid_hair_02_r {parent=dyn_mid_hair_01_r, loc=<6 0 0>, rot=<0 -0.1174 0 -0.9931>}
 dyn_mid_hair_03_r {parent=dyn_mid_hair_02_r, loc=<6 0 0>, rot=<0 -0.165 0 -0.9863>}
 mstr_face_hair_01 {parent=head, loc=<1.9881 2.534 6.5214>, rot=<0 0.1305 0.9914 0>}
 dyn_face_hair_01 {parent=mstr_face_hair_01, loc=<0 0 0>, rot=<0 0 0 1>}
 dyn_face_hair_02 {parent=dyn_face_hair_01, loc=<4 0 0>, rot=<0 0 0 1>}
 dyn_face_hair_03 {parent=dyn_face_hair_02, loc=<4 0 0>, rot=<0 0 0 1>}
 clavicle_l {parent=spine_05, loc=<5.4831 -0.1053 -3.2946>, rot=<0.0465 -0.7289 -0.009 -0.6829>}
 upperarm_l {parent=clavicle_l, loc=<9.8658 -1.2136 -0.4105>, rot=<0.0164 -0.3245 -0.0269 -0.9454>}
 clavicle_r {parent=spine_05, loc=<5.4831 -0.1053 3.2946>, rot=<0.7289 0.0465 -0.6829 0.009>}
 upperarm_r {parent=clavicle_r, loc=<-9.8658 1.2136 0.4105>, rot=<0.0164 -0.3245 -0.0269 -0.9454>}
 Aim_jnt {parent=spine_03, loc=<0 -15.3 0>, rot=<0 0 0 1>}
]
variables [
 CrouchWalk {type=double, default=0}
 PistolEquip {type=bool, default=False}
]
