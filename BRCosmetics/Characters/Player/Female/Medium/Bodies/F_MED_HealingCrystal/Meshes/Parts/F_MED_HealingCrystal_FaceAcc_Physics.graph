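# Physics asset for the HealingCrystal female medium character's FaceAcc part, covering its dynamic hair and upper body.
# Capsule bodies form middle and lower hair chains (center, left, right) on master bones, alongside head, upper spine, clavicle and upper-arm shapes.
# Constraints lock swing and twist along hair chains and head-to-spine_05.
# Clavicle and spine links swing freely, as do several hair-master links to spine_04, some duplicating locked head links.
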PhysicsAsset /BRCosmetics/Characters/Player/Female/Medium/Bodies/F_MED_HealingCrystal/Meshes/Parts/F_MED_HealingCrystal_FaceAcc_Physics
bodies {
 clavicle_l [{shapeType=Capsule, length=6.3, radius=8, shapeName=clavicle_l_capsule, center=<7.2 -1.2 -3.6>, rotation=<rot 90 0 0>}]
 clavicle_r [{shapeType=Capsule, length=6.3, radius=8, shapeName=clavicle_l_capsule, center=<-7.2 -1.2 3.6>, rotation=<rot 90 0 0>}]
 dyn_lower_hair_01 [{shapeType=Capsule, length=10, radius=4, shapeName=dyn_lower_hair_01_capsule, center=<0 0 3.4076>, rotation=<rot 0 0 90>}]
 dyn_lower_hair_01_r [{shapeType=Capsule, length=5, radius=3, shapeName=dyn_lower_hair_01_r_capsule, rotation=<rot 8.3875 0 0>}]
 dyn_lower_hair_02 [{shapeType=Capsule, length=10, radius=4, shapeName=dyn_lower_hair_02_capsule, center=<0 0 3.4076>, rotation=<rot 0 0 90>}]
 dyn_lower_hair_02_r [{shapeType=Capsule, length=2.7613, radius=3, shapeName=dyn_lower_hair_02_r_capsule, center=<2.3635 0 0.6468>, rotation=<rot 7.4317 0 0.0001>}]
 dyn_lower_hair_03 [{shapeType=Capsule, length=10, radius=4, shapeName=dyn_lower_hair_03_capsule, center=<0 0 10.7326>, rotation=<rot 0 0 90>}]
 dyn_lower_hair_03_r [{
   shapeType=Capsule
   length=4.4424
   radius=3
   shapeName=dyn_lower_hair_03_r_capsule
   center=<2.0008 0.221 1.2204>
   rotation=<rot 20.6646 -5.6613 -2.738>
  }]
 dyn_mid_hail_01_l [{shapeType=Capsule, length=4, radius=3, shapeName=dyn_mid_hail_01_l_capsule, center=<3.2901 0 -0.1833>, rotation=<rot -19.5684 0 0>}]
 dyn_mid_hail_02_l [{shapeType=Capsule, length=4, radius=3, shapeName=dyn_mid_hail_02_l_capsule, center=<3.2901 0 -0.1833>, rotation=<rot -19.5684 0 0>}]
 dyn_mid_hail_03_l [{shapeType=Capsule, length=4, radius=3, shapeName=dyn_mid_hail_03_l_capsule, center=<3.9173 0 -2.382>, rotation=<rot -19.5684 0 0>}]
 dyn_mid_hair_01 [{shapeType=Capsule, length=8, radius=4, shapeName=dyn_mid_hair_01_capsule, center=<4.5477 0 0>, rotation=<rot 0 0 90>}]
 dyn_mid_hair_01_r [{shapeType=Capsule, length=4, radius=3, shapeName=dyn_mid_hair_01_r_capsule, center=<5 -1.1003 0>, rotation=<rot -28.9077 0 0>}]
 dyn_mid_hair_02 [{shapeType=Capsule, length=8, radius=4, shapeName=dyn_mid_hair_02_capsule, center=<4.5477 0 0>, rotation=<rot 0 0 90>}]
 dyn_mid_hair_02_r [{shapeType=Capsule, length=4, radius=3, shapeName=dyn_mid_hair_02_r_capsule, center=<5.4918 1.2488 -0.2086>, rotation=<rot -18.157 0 2.0613>}]
 dyn_mid_hair_03 [{shapeType=Capsule, length=8, radius=4, shapeName=dyn_mid_hair_03_capsule, center=<4.5477 0 2.3859>, rotation=<rot 0 0 90>}]
 dyn_mid_hair_03_r [{shapeType=Capsule, length=6, radius=3, shapeName=dyn_mid_hair_03_r_capsule, center=<5.8881 1.8663 -0.5582>, rotation=<rot -16.6528 90 0>}]
 head [{shapeType=Capsule, length=4.9246, radius=5.5636, shapeName=head_capsule, center=<0.9737 -2.0126 0>, rotation=<rot 0 -90.0003 -90>}]
 mstl_mid_hail_l [{shapeType=Capsule, length=0.505, radius=0.505, shapeName=mstl_mid_hail_l_capsule, rotation=<rot 0 -90 0>}]
 mstr_lower_hair [{shapeType=Capsule, length=0.505, radius=0.505, shapeName=mstr_lower_hair_capsule, rotation=<rot 0 -90 0>}]
 mstr_lower_hair_01_l [{shapeType=Capsule, length=4, radius=2.5, shapeName=mstr_lower_hair_01_l_capsule, rotation=<rot -8.1187 0 -0.0001>}]
 mstr_lower_hair_02_l [{shapeType=Capsule, length=4, radius=2.5, shapeName=mstr_lower_hair_02_l_capsule, rotation=<rot 4.6286 0 0>}]
 mstr_lower_hair_03_l [{shapeType=Capsule, length=4, radius=2.5, shapeName=mstr_lower_hair_03_l_capsule, rotation=<rot 18.1381 0 0>}]
 mstr_lower_hair_l [{shapeType=Capsule, length=0.505, radius=0.505, shapeName=mstr_lower_hair_l_capsule, rotation=<rot 0 -90 0>}]
 mstr_lower_hair_r [{shapeType=Capsule, length=5, radius=3, shapeName=mstr_lower_hair_r_capsule, rotation=<rot 0 -90 0>}]
 mstr_mid_hair [{shapeType=Capsule, length=0.505, radius=0.505, shapeName=mstr_mid_hair_capsule, rotation=<rot 0 -90 0>}]
 mstr_mid_hair_r [{shapeType=Capsule, length=0.505, radius=0.505, shapeName=mstr_mid_hair_r_capsule, rotation=<rot 0 -90 0>}]
 spine_04 [{
   shapeType=Capsule
   length=15.5988
   radius=5.9008
   shapeName=spine_04_capsule
   center=<14.2237 -7.8448 1.8051>
   rotation=<rot 0 -90.0002 -1.4797>
  }]
 spine_05 [{shapeType=Capsule, length=18, radius=12, shapeName=spine_05_capsule, center=<2.864 -2.4931 0>, rotation=<rot 0 -90 0>}]
 upperarm_l [{shapeType=Capsule, length=26.7117, radius=4.9758, shapeName=upperarm_l_capsule, center=<16.7022 0 0>, rotation=<rot 90 0 0>}]
 upperarm_r [{shapeType=Capsule, length=26.7117, radius=4.9758, shapeName=upperarm_l_capsule, center=<-16.7022 0 0>, rotation=<rot 90 0 0>}]
}
constraints [
 {
  bone1Name=upperarm_l
  bone2Name=clavicle_l
  swing1LimitDegrees=45
  swing1Motion=Free
  swing2LimitDegrees=45
  swing2Motion=Free
  twistLimitDegrees=45
  twistMotion=Free
 }
 {
  bone1Name=clavicle_l
  bone2Name=spine_05
  swing1LimitDegrees=45
  swing1Motion=Free
  swing2LimitDegrees=45
  swing2Motion=Free
  twistLimitDegrees=45
  twistMotion=Free
 }
 {
  bone1Name=clavicle_r
  bone2Name=spine_05
  swing1LimitDegrees=45
  swing1Motion=Free
  swing2LimitDegrees=45
  swing2Motion=Free
  twistLimitDegrees=45
  twistMotion=Free
 }
 {
  bone1Name=upperarm_r
  bone2Name=clavicle_r
  swing1LimitDegrees=45
  swing1Motion=Free
  swing2LimitDegrees=45
  swing2Motion=Free
  twistLimitDegrees=45
  twistMotion=Free
 }
 {
  bone1Name=dyn_mid_hail_01_l
  bone2Name=mstl_mid_hail_l
  swing1LimitDegrees=45
  swing1Motion=Locked
  swing2LimitDegrees=45
  swing2Motion=Locked
  twistLimitDegrees=45
  twistMotion=Locked
 }
 {
  bone1Name=dyn_mid_hail_02_l
  bone2Name=dyn_mid_hail_01_l
  swing1LimitDegrees=45
  swing1Motion=Locked
  swing2LimitDegrees=45
  swing2Motion=Locked
  twistLimitDegrees=45
  twistMotion=Locked
 }
 {
  bone1Name=dyn_mid_hail_03_l
  bone2Name=dyn_mid_hail_02_l
  swing1LimitDegrees=45
  swing1Motion=Locked
  swing2LimitDegrees=45
  swing2Motion=Locked
  twistLimitDegrees=45
  twistMotion=Locked
 }
 {
  bone1Name=dyn_lower_hair_01
  bone2Name=mstr_lower_hair
  swing1LimitDegrees=45
  swing1Motion=Locked
  swing2LimitDegrees=45
  swing2Motion=Locked
  twistLimitDegrees=45
  twistMotion=Locked
 }
 {
  bone1Name=dyn_lower_hair_02
  bone2Name=dyn_lower_hair_01
  swing1LimitDegrees=45
  swing1Motion=Locked
  swing2LimitDegrees=45
  swing2Motion=Locked
  twistLimitDegrees=45
  twistMotion=Locked
 }
 {
  bone1Name=dyn_lower_hair_03
  bone2Name=dyn_lower_hair_02
  swing1LimitDegrees=45
  swing1Motion=Locked
  swing2LimitDegrees=45
  swing2Motion=Locked
  twistLimitDegrees=45
  twistMotion=Locked
 }
 {
  bone1Name=mstr_lower_hair_l
  bone2Name=mstr_lower_hair
  swing1LimitDegrees=45
  swing1Motion=Locked
  swing2LimitDegrees=45
  swing2Motion=Locked
  twistLimitDegrees=45
  twistMotion=Locked
 }
 {
  bone1Name=mstr_lower_hair_01_l
  bone2Name=mstr_lower_hair_l
  swing1LimitDegrees=45
  swing1Motion=Locked
  swing2LimitDegrees=45
  swing2Motion=Locked
  twistLimitDegrees=45
  twistMotion=Locked
 }
 {
  bone1Name=mstr_lower_hair_02_l
  bone2Name=mstr_lower_hair_01_l
  swing1LimitDegrees=45
  swing1Motion=Locked
  swing2LimitDegrees=45
  swing2Motion=Locked
  twistLimitDegrees=45
  twistMotion=Locked
 }
 {
  bone1Name=mstr_lower_hair_03_l
  bone2Name=mstr_lower_hair_02_l
  swing1LimitDegrees=45
  swing1Motion=Locked
  swing2LimitDegrees=45
  swing2Motion=Locked
  twistLimitDegrees=45
  twistMotion=Locked
 }
 {
  bone1Name=mstr_lower_hair_r
  bone2Name=mstr_lower_hair
  swing1LimitDegrees=45
  swing1Motion=Locked
  swing2LimitDegrees=45
  swing2Motion=Locked
  twistLimitDegrees=45
  twistMotion=Locked
 }
 {
  bone1Name=dyn_lower_hair_01_r
  bone2Name=mstr_lower_hair_r
  swing1LimitDegrees=45
  swing1Motion=Locked
  swing2LimitDegrees=45
  swing2Motion=Locked
  twistLimitDegrees=45
  twistMotion=Locked
 }
 {
  bone1Name=dyn_lower_hair_02_r
  bone2Name=dyn_lower_hair_01_r
  swing1LimitDegrees=20
  swing1Motion=Locked
  swing2LimitDegrees=20
  swing2Motion=Locked
  twistLimitDegrees=0
  twistMotion=Locked
 }
 {
  bone1Name=dyn_lower_hair_03_r
  bone2Name=dyn_lower_hair_02_r
  swing1LimitDegrees=45
  swing1Motion=Locked
  swing2LimitDegrees=45
  swing2Motion=Locked
  twistLimitDegrees=45
  twistMotion=Locked
 }
 {
  bone1Name=dyn_mid_hair_01
  bone2Name=mstr_mid_hair
  swing1LimitDegrees=45
  swing1Motion=Locked
  swing2LimitDegrees=45
  swing2Motion=Locked
  twistLimitDegrees=45
  twistMotion=Locked
 }
 {
  bone1Name=dyn_mid_hair_02
  bone2Name=dyn_mid_hair_01
  swing1LimitDegrees=45
  swing1Motion=Locked
  swing2LimitDegrees=45
  swing2Motion=Locked
  twistLimitDegrees=45
  twistMotion=Locked
 }
 {
  bone1Name=dyn_mid_hair_03
  bone2Name=dyn_mid_hair_02
  swing1LimitDegrees=45
  swing1Motion=Locked
  swing2LimitDegrees=45
  swing2Motion=Locked
  twistLimitDegrees=45
  twistMotion=Locked
 }
 {
  bone1Name=dyn_mid_hair_01_r
  bone2Name=mstr_mid_hair_r
  swing1LimitDegrees=45
  swing1Motion=Locked
  swing2LimitDegrees=45
  swing2Motion=Locked
  twistLimitDegrees=45
  twistMotion=Locked
 }
 {
  bone1Name=dyn_mid_hair_02_r
  bone2Name=dyn_mid_hair_01_r
  swing1LimitDegrees=45
  swing1Motion=Locked
  swing2LimitDegrees=45
  swing2Motion=Locked
  twistLimitDegrees=45
  twistMotion=Locked
 }
 {
  bone1Name=dyn_mid_hair_03_r
  bone2Name=dyn_mid_hair_02_r
  swing1LimitDegrees=45
  swing1Motion=Locked
  swing2LimitDegrees=45
  swing2Motion=Locked
  twistLimitDegrees=45
  twistMotion=Locked
 }
 {
  bone1Name=head
  bone2Name=spine_05
  swing1LimitDegrees=45
  swing1Motion=Locked
  swing2LimitDegrees=45
  swing2Motion=Locked
  twistLimitDegrees=45
  twistMotion=Locked
 }
 {
  bone1Name=upperarm_l
  bone2Name=spine_05
  swing1LimitDegrees=45
  swing1Motion=Locked
  swing2LimitDegrees=45
  swing2Motion=Locked
  twistLimitDegrees=45
  twistMotion=Locked
 }
 {
  bone1Name=upperarm_r
  bone2Name=spine_05
  swing1LimitDegrees=45
  swing1Motion=Locked
  swing2LimitDegrees=45
  swing2Motion=Locked
  twistLimitDegrees=45
  twistMotion=Locked
 }
 {
  bone1Name=mstr_lower_hair
  bone2Name=head
  swing1LimitDegrees=45
  swing1Motion=Locked
  swing2LimitDegrees=45
  swing2Motion=Locked
  twistLimitDegrees=45
  twistMotion=Locked
 }
 {
  bone1Name=mstr_mid_hair_r
  bone2Name=head
  swing1LimitDegrees=45
  swing1Motion=Locked
  swing2LimitDegrees=45
  swing2Motion=Locked
  twistLimitDegrees=45
  twistMotion=Locked
 }
 {
  bone1Name=spine_05
  bone2Name=spine_04
  swing1LimitDegrees=45
  swing1Motion=Free
  swing2LimitDegrees=45
  swing2Motion=Free
  twistLimitDegrees=45
  twistMotion=Free
 }
 {
  bone1Name=mstr_mid_hair
  bone2Name=spine_04
  swing1LimitDegrees=45
  swing1Motion=Free
  swing2LimitDegrees=45
  swing2Motion=Free
  twistLimitDegrees=45
  twistMotion=Free
 }
 {
  bone1Name=mstr_lower_hair
  bone2Name=spine_04
  swing1LimitDegrees=45
  swing1Motion=Free
  swing2LimitDegrees=45
  swing2Motion=Free
  twistLimitDegrees=45
  twistMotion=Free
 }
 {
  bone1Name=mstl_mid_hail_l
  bone2Name=spine_04
  swing1LimitDegrees=45
  swing1Motion=Free
  swing2LimitDegrees=45
  swing2Motion=Free
  twistLimitDegrees=45
  twistMotion=Free
 }
 {
  bone1Name=mstr_mid_hair_r
  bone2Name=spine_04
  swing1LimitDegrees=45
  swing1Motion=Free
  swing2LimitDegrees=45
  swing2Motion=Free
  twistLimitDegrees=45
  twistMotion=Free
 }
]
