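# Post-pose AnimBP for the female HealingCrystal face accessory, adding secondary motion to its hair bones.
# The copied parent pose passes through a Control Rig fed the pistol-aim flag, RigidBody, a weighted constraint and per-bone AnimDynamics chains.
# Additive bone offsets finish the pose, smoothed by alpha interpolation.
# EventGraph derives pistol and rifle/shotgun aim states, head-up alphas from spine-to-head twist, and a gravity override.
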AnimBlueprint /BRCosmetics/Characters/Player/Female/Medium/Bodies/F_MED_HealingCrystal/Meshes/Parts/F_MED_HealingCrystal_FaceAcc_AnimBP
target_skeleton=&/F_MED_HealingCrystal_FaceAcc_Skeleton
graphs [
 AnimGraph {nodes [
   "Transform (Modify) Bone - Bone: dyn_mid_hair_03_r"(Alpha=1, Translation="0,-1,0", Rotation="50,0,1", Scale="1,1,1") {
    rotationMode=BMM_Additive
    translationMode=BMM_Additive
    alphaBoolBlend {blendOption=Linear}
    alphaScaleBiasClamp {bInterpResult=true}
    boneToModify {boneName=dyn_mid_hair_03_r}
    rotation {pitch=50, roll=1}
    translation {y=-1}
   }
   n1 = Control Rig!(Source=Copy Pose From Mesh() {bUseAttachedParent=true}, Alpha, PistolEquip:bool=Get Pistol Equipped() -> (Pistol Equipped:bool)) -> (Pose) {
    alphaBoolBlend {blendOption=Linear}
    controlRigAssetReference {blueprintRigClass=&/F_MED_HealingCrystal_FaceAcc_CtrlRig.F_MED_HealingCrystal_FaceAcc_CtrlRig_C}
   }
   n2 = RigidBody!(ComponentPose=Local To Component(LocalPose=@n1), Alpha, OverrideWorldGravity:Vector="0,0,0", ExternalForce) -> (Pose) {
    bClampLinearTranslationLimitToRefPose=true
    bOverrideWorldGravity=true
    overridePhysicsAsset=&/F_MED_HealingCrystal_FaceAcc_Physics
    simulationSpace=BaseBoneSpace
    alphaBoolBlend {blendOption=Linear}
    baseBoneRef {boneName=spine_05}
    componentAppliedLinearAccClamp=<3000 3000 3000>
    componentLinearAccScale=<0.25 0.1 0.1>
    componentLinearVelScale=<0.25 0.1 0.1>
    simSpaceSettings {
     maxAngularAcceleration=1000
     maxAngularVelocity=3
     maxLinearAcceleration=700
     maxLinearVelocity=500
     velocityScaleZ=0
     worldAlpha=0.5
     externalLinearDragV=<0.2 0.2>
    }
   }
   n3 = "Constraint - dyn_lower_hair_02_r (RRS)"!(ComponentPose=@n2, Alpha=1, ConstraintWeights_0=0.3, ConstraintWeights_1:float=0.5, ConstraintWeights_2:float=0.2) -> (Pose) {
    alphaBoolBlend {blendOption=Linear}
    boneToModify {boneName=dyn_lower_hair_02_r}
    constraintSetup [
     {transformType=Rotation, targetBone {boneName=upperarm_r}}
     {transformType=Rotation, targetBone {boneName=dyn_mid_hair_01_r}}
     {transformType=Scale, targetBone {boneName=head}}
    ]
    constraintWeights [1,1,1]
   }
   n4 = Anim Dynamics - Bone: dyn_top_hair_01!(ComponentPose=@n3, Alpha=1, GravityOverride:Vector="0,0,-980", ExternalForce="0,0,0") -> (Pose) {
    angularDampingOverride=1
    angularSpringConstant=1200
    bAngularSpring=true
    bOverrideAngularDamping=true
    bUseGravityOverride=true
    gravityScale=0.35
    simulationSpace=RootRelative
    alphaBoolBlend {blendOption=Linear}
    boundBone {boneName=dyn_top_hair_01}
    componentAppliedLinearAccClamp=<4000 4000 4000>
    componentLinearAccScale=<0.2 0.2 0.2>
    componentLinearVelScale=<0.2 0.2 0.2>
    gravityOverride {z=-980}
    physicsBodyDefinitions [{
      boundBone {boneName=dyn_top_hair_01}
      boxExtents=<12 12 12>
      constraintSetup {angularLimitsMax=<0 2 1>, angularLimitsMin=<0 -5 -1>, angularTarget=<1 0 0>}
      localJointOffset=<1 0 0>
     }]
   }
   n5 = Anim Dynamics - Bone: dyn_top_hair_02!(ComponentPose=@n4, Alpha=1, GravityOverride:Vector="0,0,-980", ExternalForce="0,0,0") -> (Pose) {
    angularDampingOverride=1
    angularSpringConstant=1200
    bAngularSpring=true
    bOverrideAngularDamping=true
    bUseGravityOverride=true
    gravityScale=0.35
    simulationSpace=RootRelative
    alphaBoolBlend {blendOption=Linear}
    boundBone {boneName=dyn_top_hair_02}
    componentAppliedLinearAccClamp=<4000 4000 4000>
    componentLinearAccScale=<0.2 0.2 0.2>
    componentLinearVelScale=<0.2 0.2 0.2>
    gravityOverride {z=-980}
    physicsBodyDefinitions [{
      boundBone {boneName=dyn_top_hair_02}
      boxExtents=<12 12 12>
      constraintSetup {angularLimitsMax=<0 2 1>, angularLimitsMin=<0 -5 -1>, angularTarget=<1 0 0>}
      localJointOffset=<1 0 0>
     }]
   }
   n6 = Anim Dynamics - Bone: dyn_top_hair_03!(ComponentPose=@n5, Alpha=1, GravityOverride:Vector="0,0,-980", ExternalForce="0,0,0") -> (Pose) {
    angularDampingOverride=1
    angularSpringConstant=1200
    bAngularSpring=true
    bOverrideAngularDamping=true
    bUseGravityOverride=true
    gravityScale=0.35
    simulationSpace=RootRelative
    alphaBoolBlend {blendOption=Linear}
    boundBone {boneName=dyn_top_hair_03}
    componentAppliedLinearAccClamp=<4000 4000 4000>
    componentLinearAccScale=<0.2 0.2 0.2>
    componentLinearVelScale=<0.2 0.2 0.2>
    gravityOverride {z=-980}
    physicsBodyDefinitions [{
      boundBone {boneName=dyn_top_hair_03}
      boxExtents=<12 12 12>
      constraintSetup {angularLimitsMax=<0 2 1>, angularLimitsMin=<0 -5 -1>, angularTarget=<1 0 0>}
      localJointOffset=<1 0 0>
     }]
   }
   n7 = Anim Dynamics - Bone: dyn_mid_hair_01!(ComponentPose=@n6, Alpha=1, GravityOverride:Vector="0,0,-980", ExternalForce="0,0,0") -> (Pose) {
    angularDampingOverride=1
    angularSpringConstant=1200
    bAngularSpring=true
    bOverrideAngularDamping=true
    bUseGravityOverride=true
    gravityScale=0.35
    simulationSpace=RootRelative
    alphaBoolBlend {blendOption=Linear}
    boundBone {boneName=dyn_mid_hair_01}
    componentAppliedLinearAccClamp=<4000 4000 4000>
    componentLinearAccScale=<0.2 0.2 0.2>
    componentLinearVelScale=<0.2 0.2 0.2>
    gravityOverride {z=-980}
    physicsBodyDefinitions [{
      boundBone {boneName=dyn_mid_hair_01}
      boxExtents=<12 12 12>
      constraintSetup {angularLimitsMax=<0 2 1>, angularLimitsMin=<0 -5 -1>, angularTarget=<1 0 0>}
      localJointOffset=<1 0 0>
     }]
   }
   n8 = Anim Dynamics - Bone: dyn_mid_hair_02!(ComponentPose=@n7, Alpha=1, GravityOverride:Vector="0,0,-980", ExternalForce="0,0,0") -> (Pose) {
    angularDampingOverride=1
    angularSpringConstant=1200
    bAngularSpring=true
    bOverrideAngularDamping=true
    bUseGravityOverride=true
    gravityScale=0.35
    simulationSpace=RootRelative
    alphaBoolBlend {blendOption=Linear}
    boundBone {boneName=dyn_mid_hair_02}
    componentAppliedLinearAccClamp=<4000 4000 4000>
    componentLinearAccScale=<0.2 0.2 0.2>
    componentLinearVelScale=<0.2 0.2 0.2>
    gravityOverride {z=-980}
    physicsBodyDefinitions [{
      boundBone {boneName=dyn_mid_hair_02}
      boxExtents=<12 12 12>
      constraintSetup {angularLimitsMax=<0 2 1>, angularLimitsMin=<0 -5 -1>, angularTarget=<1 0 0>}
      localJointOffset=<1 0 0>
     }]
   }
   n9 = Anim Dynamics - Bone: dyn_mid_hair_03!(ComponentPose=@n8, Alpha=1, GravityOverride:Vector="0,0,-980", ExternalForce="0,0,0") -> (Pose) {
    angularDampingOverride=1
    angularSpringConstant=1200
    bAngularSpring=true
    bOverrideAngularDamping=true
    bUseGravityOverride=true
    gravityScale=0.35
    simulationSpace=RootRelative
    alphaBoolBlend {blendOption=Linear}
    boundBone {boneName=dyn_mid_hair_03}
    componentAppliedLinearAccClamp=<4000 4000 4000>
    componentLinearAccScale=<0.2 0.2 0.2>
    componentLinearVelScale=<0.2 0.2 0.2>
    gravityOverride {z=-980}
    physicsBodyDefinitions [{
      boundBone {boneName=dyn_mid_hair_03}
      boxExtents=<12 12 12>
      constraintSetup {angularLimitsMax=<0 2 1>, angularLimitsMin=<0 -5 -1>, angularTarget=<1 0 0>}
      localJointOffset=<1 0 0>
     }]
   }
   n10 = Anim Dynamics - Bone: dyn_lower_hair_01!(ComponentPose=@n9, Alpha=1, GravityOverride:Vector="0,0,-980", ExternalForce="0,0,0") -> (Pose) {
    angularDampingOverride=1
    angularSpringConstant=1200
    bAngularSpring=true
    bOverrideAngularDamping=true
    bUseGravityOverride=true
    gravityScale=0.35
    simulationSpace=RootRelative
    alphaBoolBlend {blendOption=Linear}
    boundBone {boneName=dyn_lower_hair_01}
    componentAppliedLinearAccClamp=<4000 4000 4000>
    componentLinearAccScale=<0.2 0.2 0.2>
    componentLinearVelScale=<0.2 0.2 0.2>
    gravityOverride {z=-980}
    physicsBodyDefinitions [{
      boundBone {boneName=dyn_lower_hair_01}
      boxExtents=<12 12 12>
      constraintSetup {angularLimitsMax=<0 2 5>, angularLimitsMin=<0 -5 -5>, angularTarget=<1 0 0>}
      localJointOffset=<1 0 0>
     }]
   }
   n11 = Anim Dynamics - Bone: dyn_lower_hair_02!(ComponentPose=@n10, Alpha=1, GravityOverride:Vector="0,0,-980", ExternalForce="0,0,0") -> (Pose) {
    angularDampingOverride=1
    angularSpringConstant=1200
    bAngularSpring=true
    bOverrideAngularDamping=true
    bUseGravityOverride=true
    bUseSphericalLimits=true
    gravityScale=0.35
    simulationSpace=RootRelative
    alphaBoolBlend {blendOption=Linear}
    boundBone {boneName=dyn_lower_hair_02}
    componentAppliedLinearAccClamp=<4000 4000 4000>
    componentLinearAccScale=<0.2 0.2 0.2>
    componentLinearVelScale=<0.2 0.2 0.2>
    gravityOverride {z=-980}
    physicsBodyDefinitions [{
      boundBone {boneName=dyn_lower_hair_02}
      boxExtents=<12 12 12>
      constraintSetup {angularLimitsMax=<0 2 0>, angularLimitsMin=<0 -2 0>, angularTarget=<1 0 0>}
      localJointOffset=<1 0 0>
     }]
    sphericalLimits [{limitRadius=22, drivingBone {boneName=spine_05}, sphereLocalOffset=<0 0 0>}]
   }
   n12 = Anim Dynamics - Bone: dyn_lower_hair_03!(ComponentPose=@n11, Alpha=1, GravityOverride:Vector="0,0,-980", ExternalForce="0,0,0") -> (Pose) {
    angularDampingOverride=1
    angularSpringConstant=1200
    bAngularSpring=true
    bOverrideAngularDamping=true
    bUseGravityOverride=true
    bUseSphericalLimits=true
    gravityScale=0.35
    simulationSpace=RootRelative
    alphaBoolBlend {blendOption=Linear}
    boundBone {boneName=dyn_lower_hair_03}
    componentAppliedLinearAccClamp=<4000 4000 4000>
    componentLinearAccScale=<0.2 0.2 0.2>
    componentLinearVelScale=<0.2 0.2 0.2>
    gravityOverride {z=-980}
    physicsBodyDefinitions [{
      boundBone {boneName=dyn_lower_hair_03}
      boxExtents=<12 12 12>
      constraintSetup {angularLimitsMax=<0 2 0>, angularLimitsMin=<0 -2 0>, angularTarget=<1 0 0>}
      localJointOffset=<1 0 0>
     }]
    sphericalLimits [{limitRadius=17, drivingBone {boneName=spine_05}, sphereLocalOffset=<0 0 0>}]
   }
   n13 = Anim Dynamics - Bone: dyn_mid_hair_01_r!(ComponentPose=@n12, Alpha=1, GravityOverride:Vector="0,0,-980", ExternalForce="0,0,0") -> (Pose) {
    angularDampingOverride=1
    angularSpringConstant=1200
    bAngularSpring=true
    bOverrideAngularDamping=true
    bUseGravityOverride=true
    gravityScale=0.35
    simulationSpace=RootRelative
    alphaBoolBlend {blendOption=Linear}
    boundBone {boneName=dyn_mid_hair_01_r}
    componentAppliedLinearAccClamp=<4000 4000 4000>
    componentLinearAccScale=<0.2 0.2 0.2>
    componentLinearVelScale=<0.2 0.2 0.2>
    gravityOverride {z=-980}
    physicsBodyDefinitions [{
      boundBone {boneName=dyn_mid_hair_01_r}
      boxExtents=<12 12 12>
      constraintSetup {angularLimitsMax=<0 1 1>, angularLimitsMin=<0 -1 -1>, angularTarget=<1 0 0>}
      localJointOffset=<1 0 0>
     }]
   }
   n14 = Anim Dynamics - Bone: dyn_mid_hair_02_r!(ComponentPose=@n13, Alpha=1, GravityOverride:Vector="0,0,-980", ExternalForce="0,0,0") -> (Pose) {
    angularDampingOverride=1
    angularSpringConstant=1200
    bAngularSpring=true
    bOverrideAngularDamping=true
    bUseGravityOverride=true
    bUseSphericalLimits=true
    gravityScale=0.35
    simulationSpace=RootRelative
    alphaBoolBlend {blendOption=Linear}
    boundBone {boneName=dyn_mid_hair_02_r}
    componentAppliedLinearAccClamp=<4000 4000 4000>
    componentLinearAccScale=<0.2 0.2 0.2>
    componentLinearVelScale=<0.2 0.2 0.2>
    gravityOverride {z=-980}
    physicsBodyDefinitions [{
      boundBone {boneName=dyn_mid_hair_02_r}
      boxExtents=<12 12 12>
      constraintSetup {angularLimitsMax=<0 1 1>, angularLimitsMin=<0 -1 -1>, angularTarget=<1 0 0>}
      localJointOffset=<1 0 0>
     }]
    sphericalLimits [
     {limitRadius=10, drivingBone {boneName=None}, sphereLocalOffset=<-8 0 0>}
     {limitRadius=7, drivingBone {boneName=None}, sphereLocalOffset=<-5 0 0>}
    ]
   }
   n15 = Anim Dynamics - Bone: dyn_mid_hair_03_r!(ComponentPose=@n14, Alpha=1, GravityOverride:Vector="0,0,-980", ExternalForce="0,0,0") -> (Pose) {
    angularDampingOverride=1
    angularSpringConstant=1200
    bAngularSpring=true
    bOverrideAngularDamping=true
    bUseGravityOverride=true
    bUseSphericalLimits=true
    gravityScale=0.35
    simulationSpace=RootRelative
    alphaBoolBlend {blendOption=Linear}
    boundBone {boneName=dyn_mid_hair_03_r}
    componentAppliedLinearAccClamp=<4000 4000 4000>
    componentLinearAccScale=<0.2 0.2 0.2>
    componentLinearVelScale=<0.2 0.2 0.2>
    gravityOverride {z=-980}
    physicsBodyDefinitions [{
      boundBone {boneName=dyn_mid_hair_03_r}
      boxExtents=<12 12 12>
      constraintSetup {angularLimitsMax=<0 1 1>, angularLimitsMin=<0 -1 -1>, angularTarget=<1 0 0>}
      localJointOffset=<1 0 0>
     }]
    sphericalLimits [
     {limitRadius=6, drivingBone {boneName=None}, sphereLocalOffset=<-5 0 0>}
     {limitRadius=7, drivingBone {boneName=None}, sphereLocalOffset=<-5 -0.7959 5.0529>}
    ]
   }
   n16 = Anim Dynamics - Bone: dyn_mid_hail_01_l!(ComponentPose=@n15, Alpha=1, GravityOverride:Vector="0,0,-980", ExternalForce="0,0,0") -> (Pose) {
    angularDampingOverride=1
    angularSpringConstant=1200
    bAngularSpring=true
    bOverrideAngularDamping=true
    bUseGravityOverride=true
    gravityScale=0.35
    simulationSpace=RootRelative
    alphaBoolBlend {blendOption=Linear}
    boundBone {boneName=dyn_mid_hail_01_l}
    componentAppliedLinearAccClamp=<4000 4000 4000>
    componentLinearAccScale=<0.2 0.2 0.2>
    componentLinearVelScale=<0.2 0.2 0.2>
    gravityOverride {z=-980}
    physicsBodyDefinitions [{
      boundBone {boneName=dyn_mid_hail_01_l}
      boxExtents=<8 8 8>
      constraintSetup {angularLimitsMax=<0 1 1>, angularLimitsMin=<0 -1 -1>, angularTarget=<1 0 0>}
      localJointOffset=<1 0 0>
     }]
   }
   n17 = Anim Dynamics - Bone: dyn_mid_hail_02_l!(ComponentPose=@n16, Alpha=1, GravityOverride:Vector="0,0,-980", ExternalForce="0,0,0") -> (Pose) {
    angularDampingOverride=1
    angularSpringConstant=1200
    bAngularSpring=true
    bOverrideAngularDamping=true
    bUseGravityOverride=true
    bUseSphericalLimits=true
    gravityScale=0.35
    simulationSpace=RootRelative
    alphaBoolBlend {blendOption=Linear}
    boundBone {boneName=dyn_mid_hail_02_l}
    componentAppliedLinearAccClamp=<4000 4000 4000>
    componentLinearAccScale=<0.2 0.2 0.2>
    componentLinearVelScale=<0.2 0.2 0.2>
    gravityOverride {z=-980}
    physicsBodyDefinitions [{
      boundBone {boneName=dyn_mid_hail_02_l}
      boxExtents=<8 8 8>
      constraintSetup {angularLimitsMax=<0 1 1>, angularLimitsMin=<0 -1 -1>, angularTarget=<1 0 0>}
      localJointOffset=<1 0 0>
     }]
   }
   n18 = Anim Dynamics - Bone: dyn_mid_hail_03_l!(ComponentPose=@n17, Alpha=1, GravityOverride:Vector="0,0,-980", ExternalForce="0,0,0") -> (Pose) {
    angularDampingOverride=1
    angularSpringConstant=1200
    bAngularSpring=true
    bOverrideAngularDamping=true
    bUseGravityOverride=true
    bUseSphericalLimits=true
    gravityScale=0.35
    simulationSpace=RootRelative
    alphaBoolBlend {blendOption=Linear}
    boundBone {boneName=dyn_mid_hail_03_l}
    componentAppliedLinearAccClamp=<4000 4000 4000>
    componentLinearAccScale=<0.2 0.2 0.2>
    componentLinearVelScale=<0.2 0.2 0.2>
    gravityOverride {z=-980}
    physicsBodyDefinitions [{
      boundBone {boneName=dyn_mid_hail_03_l}
      boxExtents=<8 8 8>
      constraintSetup {angularLimitsMax=<0 1 1>, angularLimitsMin=<0 -1 -1>, angularTarget=<1 0 0>}
      localJointOffset=<1 0 0>
     }]
   }
   n19 = Anim Dynamics - Bone: dyn_lower_hair_01_r!(ComponentPose=@n18, Alpha=1, GravityOverride:Vector="0,0,-980", ExternalForce="0,0,0") -> (Pose) {
    angularDampingOverride=1
    angularSpringConstant=1200
    bAngularSpring=true
    bOverrideAngularDamping=true
    bUseGravityOverride=true
    gravityScale=0.35
    simulationSpace=RootRelative
    alphaBoolBlend {blendOption=Linear}
    boundBone {boneName=dyn_lower_hair_01_r}
    componentAppliedLinearAccClamp=<4000 4000 4000>
    componentLinearAccScale=<0.2 0.2 0.2>
    componentLinearVelScale=<0.2 0.2 0.2>
    gravityOverride {z=-980}
    physicsBodyDefinitions [{
      boundBone {boneName=dyn_lower_hair_01_r}
      boxExtents=<8 8 8>
      constraintSetup {angularLimitsMax=<0 1 1>, angularLimitsMin=<0 -1 -1>, angularTarget=<1 0 0>}
      localJointOffset=<1 0 0>
     }]
   }
   n20 = Anim Dynamics - Bone: dyn_lower_hair_02_r!(ComponentPose=@n19, Alpha=1, GravityOverride:Vector="0,0,-980", ExternalForce="0,0,0") -> (Pose) {
    angularDampingOverride=1
    angularSpringConstant=1200
    bAngularSpring=true
    bOverrideAngularDamping=true
    bUseGravityOverride=true
    gravityScale=0.35
    simulationSpace=RootRelative
    alphaBoolBlend {blendOption=Linear}
    boundBone {boneName=dyn_lower_hair_02_r}
    componentAppliedLinearAccClamp=<4000 4000 4000>
    componentLinearAccScale=<0.2 0.2 0.2>
    componentLinearVelScale=<0.2 0.2 0.2>
    gravityOverride {z=-980}
    physicsBodyDefinitions [{
      boundBone {boneName=dyn_lower_hair_02_r}
      boxExtents=<8 8 8>
      constraintSetup {angularLimitsMax=<0 1 1>, angularLimitsMin=<0 -1 -1>, angularTarget=<1 0 0>}
      localJointOffset=<1 0 0>
     }]
    sphericalLimits [{limitRadius=8, drivingBone {boneName=upperarm_r}, sphereLocalOffset=<0 0 0>}]
   }
   n21 = Anim Dynamics - Bone: dyn_lower_hair_03_r!(ComponentPose=@n20, Alpha=1, GravityOverride:Vector="0,0,-980", ExternalForce="0,0,0") -> (Pose) {
    angularDampingOverride=1
    angularSpringConstant=1200
    bAngularSpring=true
    bOverrideAngularDamping=true
    bUseGravityOverride=true
    gravityScale=0.35
    simulationSpace=RootRelative
    alphaBoolBlend {blendOption=Linear}
    boundBone {boneName=dyn_lower_hair_03_r}
    componentAppliedLinearAccClamp=<4000 4000 4000>
    componentLinearAccScale=<0.2 0.2 0.2>
    componentLinearVelScale=<0.2 0.2 0.2>
    gravityOverride {z=-980}
    physicsBodyDefinitions [{
      boundBone {boneName=dyn_lower_hair_03_r}
      boxExtents=<8 8 8>
      constraintSetup {angularLimitsMax=<0 1 1>, angularLimitsMin=<0 -1 -1>, angularTarget=<1 0 0>}
      localJointOffset=<1 0 0>
     }]
    sphericalLimits [{limitRadius=6, drivingBone {boneName=upperarm_r}, sphereLocalOffset=<0 0 0>}]
   }
   n22 = Anim Dynamics - Bone: mstr_lower_hair_01_l!(ComponentPose=@n21, Alpha=1, GravityOverride:Vector="0,0,-980", ExternalForce="0,0,0") -> (Pose) {
    angularDampingOverride=1
    angularSpringConstant=1200
    bAngularSpring=true
    bOverrideAngularDamping=true
    bUseGravityOverride=true
    gravityScale=0.35
    simulationSpace=RootRelative
    alphaBoolBlend {blendOption=Linear}
    boundBone {boneName=mstr_lower_hair_01_l}
    componentAppliedLinearAccClamp=<4000 4000 4000>
    componentLinearAccScale=<0.2 0.2 0.2>
    componentLinearVelScale=<0.2 0.2 0.2>
    gravityOverride {z=-980}
    physicsBodyDefinitions [{
      boundBone {boneName=mstr_lower_hair_01_l}
      boxExtents=<8 8 8>
      constraintSetup {angularLimitsMax=<0 1 1>, angularLimitsMin=<0 -1 -1>, angularTarget=<1 0 0>}
      localJointOffset=<1 0 0>
     }]
   }
   n23 = Anim Dynamics - Bone: mstr_lower_hair_02_l!(ComponentPose=@n22, Alpha=1, GravityOverride:Vector="0,0,-980", ExternalForce="0,0,0") -> (Pose) {
    angularDampingOverride=1
    angularSpringConstant=1200
    bAngularSpring=true
    bOverrideAngularDamping=true
    bUseGravityOverride=true
    gravityScale=0.35
    simulationSpace=RootRelative
    alphaBoolBlend {blendOption=Linear}
    boundBone {boneName=mstr_lower_hair_02_l}
    componentAppliedLinearAccClamp=<4000 4000 4000>
    componentLinearAccScale=<0.2 0.2 0.2>
    componentLinearVelScale=<0.2 0.2 0.2>
    gravityOverride {z=-980}
    physicsBodyDefinitions [{
      boundBone {boneName=mstr_lower_hair_02_l}
      boxExtents=<8 8 8>
      constraintSetup {angularLimitsMax=<0 1 1>, angularLimitsMin=<0 -1 -1>, angularTarget=<1 0 0>}
      localJointOffset=<1 0 0>
     }]
   }
   n24 = Anim Dynamics - Bone: mstr_lower_hair_03_l!(ComponentPose=@n23, Alpha=1, GravityOverride:Vector="0,0,-980", ExternalForce="0,0,0") -> (Pose) {
    angularDampingOverride=1
    angularSpringConstant=1200
    bAngularSpring=true
    bOverrideAngularDamping=true
    bUseGravityOverride=true
    gravityScale=0.35
    simulationSpace=RootRelative
    alphaBoolBlend {blendOption=Linear}
    boundBone {boneName=mstr_lower_hair_03_l}
    componentAppliedLinearAccClamp=<4000 4000 4000>
    componentLinearAccScale=<0.2 0.2 0.2>
    componentLinearVelScale=<0.2 0.2 0.2>
    gravityOverride {z=-980}
    physicsBodyDefinitions [{
      boundBone {boneName=mstr_lower_hair_03_l}
      boxExtents=<8 8 8>
      constraintSetup {angularLimitsMax=<0 1 1>, angularLimitsMin=<0 -1 -1>, angularTarget=<1 0 0>}
      localJointOffset=<1 0 0>
     }]
   }
   n25 = Anim Dynamics - Bone: dyn_face_hair_01!(ComponentPose=@n24, Alpha=1, GravityOverride:Vector="0,0,-980", ExternalForce="0,0,0") -> (Pose) {
    angularDampingOverride=1
    angularSpringConstant=2000
    bAngularSpring=true
    bOverrideAngularDamping=true
    bUseGravityOverride=true
    gravityScale=0.35
    simulationSpace=RootRelative
    alphaBoolBlend {blendOption=Linear}
    boundBone {boneName=dyn_face_hair_01}
    componentLinearAccScale=<1 1 1>
    componentLinearVelScale=<1 1 1>
    gravityOverride {z=-980}
    physicsBodyDefinitions [{
      boundBone {boneName=dyn_face_hair_01}
      boxExtents=<10 10 10>
      constraintSetup {angularLimitsMax=<0 2 2>, angularLimitsMin=<0 -2 -2>, angularTarget=<1 0 0>}
      localJointOffset=<4 0 0>
     }]
   }
   n26 = Anim Dynamics - Bone: dyn_face_hair_02!(ComponentPose=@n25, Alpha=1, GravityOverride:Vector="0,0,-980", ExternalForce="0,0,0") -> (Pose) {
    angularDampingOverride=1
    angularSpringConstant=1000
    bAngularSpring=true
    bOverrideAngularDamping=true
    bUseGravityOverride=true
    gravityScale=0.35
    simulationSpace=RootRelative
    alphaBoolBlend {blendOption=Linear}
    boundBone {boneName=dyn_face_hair_02}
    componentLinearAccScale=<1 1 1>
    componentLinearVelScale=<1 1 1>
    gravityOverride {z=-980}
    physicsBodyDefinitions [{
      boundBone {boneName=dyn_face_hair_02}
      boxExtents=<10 10 10>
      constraintSetup {angularLimitsMax=<0 3 3>, angularLimitsMin=<0 -3 -3>, angularTarget=<1 0 0>}
      localJointOffset=<4 0 0>
     }]
   }
   n27 = Anim Dynamics - Bone: dyn_face_hair_03!(ComponentPose=@n26, Alpha=1, GravityOverride:Vector="0,0,-980", ExternalForce="0,0,0") -> (Pose) {
    angularDampingOverride=1
    angularSpringConstant=1000
    bAngularSpring=true
    bOverrideAngularDamping=true
    bUseGravityOverride=true
    gravityScale=0.35
    simulationSpace=RootRelative
    alphaBoolBlend {blendOption=Linear}
    boundBone {boneName=dyn_face_hair_03}
    componentLinearAccScale=<1 1 1>
    componentLinearVelScale=<1 1 1>
    gravityOverride {z=-980}
    physicsBodyDefinitions [{
      boundBone {boneName=dyn_face_hair_03}
      boxExtents=<10 10 10>
      constraintSetup {angularLimitsMax=<0 5 5>, angularLimitsMin=<0 -5 -5>, angularTarget=<1 0 0>}
      localJointOffset=<4 0 0>
     }]
   }
   n28 = "Transform (Modify) Bone - Bone: mstr_lower_hair_r"(ComponentPose=@n27, Alpha=1, Translation="0,0,0", Rotation="0,0,0", Scale="1,1,1") {
    rotationMode=BMM_Additive
    translationMode=BMM_Additive
    alphaBoolBlend {blendOption=Linear}
    alphaScaleBiasClamp {bClampResult=true, bInterpResult=true}
    boneToModify {boneName=mstr_lower_hair_r}
   }
   n29 = "Transform (Modify) Bone - Bone: dyn_lower_hair_02_r"(ComponentPose=@n28, Alpha=1, Translation="0,0,0", Rotation="0,0,0", Scale="1,1,1") {
    rotationMode=BMM_Additive
    translationMode=BMM_Additive
    alphaBoolBlend {blendOption=Linear}
    alphaScaleBiasClamp {bClampResult=true, bInterpResult=true}
    boneToModify {boneName=dyn_lower_hair_02_r}
   }
   n30 = "Transform (Modify) Bone - Bone: dyn_mid_hair_01_r"(ComponentPose=@n29, Alpha=1, Translation="0,0,0", Rotation="0,0,0", Scale="1,1,1") {
    rotationMode=BMM_Additive
    translationMode=BMM_Additive
    alphaBoolBlend {blendOption=Linear}
    alphaScaleBiasClamp {bClampResult=true, bInterpResult=true, interpSpeedDecreasing=5, interpSpeedIncreasing=8}
    boneToModify {boneName=dyn_mid_hair_01_r}
   }
   n31 = "Transform (Modify) Bone - Bone: mstr_mid_hair"(ComponentPose=@n30, Alpha=1, Translation="0,0,0", Rotation="0,10,0", Scale="1,1,1") {
    rotationMode=BMM_Additive
    translationMode=BMM_Additive
    alphaBoolBlend {blendOption=Linear}
    alphaScaleBiasClamp {bClampResult=true, bInterpResult=true, interpSpeedIncreasing=8}
    boneToModify {boneName=mstr_mid_hair}
    rotation {yaw=10}
   }
   n32 = "Transform (Modify) Bone - Bone: dyn_lower_hair_01_r"(ComponentPose=@n31, Alpha=1, Translation="1,0,0", Rotation="0,0,0", Scale="1,1,1") {
    rotationMode=BMM_Additive
    translationMode=BMM_Additive
    alphaBoolBlend {blendOption=Linear}
    alphaScaleBiasClamp {bClampResult=true, bInterpResult=true, interpSpeedDecreasing=7, interpSpeedIncreasing=3}
    boneToModify {boneName=dyn_lower_hair_01_r}
    translation {x=1}
   }
   n33 = "Transform (Modify) Bone - Bone: dyn_lower_hair_02_r"(ComponentPose=@n32, Alpha=1, Translation="0,0,0", Rotation="0,30,-30", Scale="1,1,1") {
    rotationMode=BMM_Additive
    translationMode=BMM_Additive
    alphaBoolBlend {blendOption=Linear}
    alphaScaleBiasClamp {bClampResult=true, bInterpResult=true}
    boneToModify {boneName=dyn_lower_hair_02_r}
    rotation {roll=-30, yaw=30}
   }
   n34 = "Transform (Modify) Bone - Bone: dyn_lower_hair_03_r"(ComponentPose=@n33, Alpha=1, Translation="0,0,0", Rotation="0,20,0", Scale="1,1,1") {
    rotationMode=BMM_Additive
    translationMode=BMM_Additive
    alphaBoolBlend {blendOption=Linear}
    alphaScaleBiasClamp {bClampResult=true, bInterpResult=true, interpSpeedIncreasing=8}
    boneToModify {boneName=dyn_lower_hair_03_r}
    rotation {yaw=20}
   }
   n35 = "Transform (Modify) Bone - Bone: dyn_mid_hair_01_r"(ComponentPose=@n34, Alpha=1, Translation="0,-1.5,0", Rotation="20,0,5", Scale="1,1,1") {
    rotationMode=BMM_Additive
    translationMode=BMM_Additive
    alphaBoolBlend {blendOption=Linear}
    alphaScaleBiasClamp {bInterpResult=true}
    boneToModify {boneName=dyn_mid_hair_01_r}
    rotation {pitch=20, roll=5}
    translation {y=-1.5}
   }
   n36 = "Transform (Modify) Bone - Bone: dyn_mid_hair_02_r"(ComponentPose=@n35, Alpha=1, Translation="0,-1.5,0", Rotation="10,0,0", Scale="1,1,1") {
    rotationMode=BMM_Additive
    translationMode=BMM_Additive
    alphaBoolBlend {blendOption=Linear}
    alphaScaleBiasClamp {bInterpResult=true}
    boneToModify {boneName=dyn_mid_hair_02_r}
    rotation {pitch=10}
    translation {y=-1.5}
   }
   n37 = "Transform (Modify) Bone - Bone: mstr_mid_hair_r"(ComponentPose=@n36, Alpha=1, Translation="0,0,0", Rotation="0,0,10", Scale="1,1,1") {
    rotationMode=BMM_Additive
    translationMode=BMM_Additive
    alphaBoolBlend {blendOption=Linear}
    alphaScaleBiasClamp {bClampResult=true, bInterpResult=true, interpSpeedDecreasing=7, interpSpeedIncreasing=7}
    boneToModify {boneName=mstr_mid_hair_r}
    rotation {roll=10}
   }
   n38 = "Transform (Modify) Bone - Bone: dyn_lower_hair_01_r"(ComponentPose=@n37, Alpha=1, Translation="0,0,0", Rotation="0,0,-10", Scale="1,1,1") {
    rotationMode=BMM_Additive
    translationMode=BMM_Additive
    alphaBoolBlend {blendOption=Linear}
    alphaScaleBiasClamp {bClampResult=true, bInterpResult=true, interpSpeedDecreasing=7, interpSpeedIncreasing=7}
    boneToModify {boneName=dyn_lower_hair_01_r}
    rotation {roll=-10}
   }
   n39 = "Transform (Modify) Bone - Bone: dyn_mid_hair_02_r"(ComponentPose=@n38, Alpha=1, Translation="0,0,0", Rotation="0,0,0", Scale="1,1,1") {
    rotationMode=BMM_Additive
    translationMode=BMM_Additive
    alphaBoolBlend {blendOption=Linear}
    alphaScaleBiasClamp {bClampResult=true, bInterpResult=true, interpSpeedDecreasing=7, interpSpeedIncreasing=7}
    boneToModify {boneName=dyn_mid_hair_02_r}
   }
   n40 = "Transform (Modify) Bone - Bone: dyn_lower_hair_02_r"(ComponentPose=@n39, Alpha=1, Translation="0,0,0", Rotation="0,0,10", Scale="1,1,1") {
    rotationMode=BMM_Additive
    translationMode=BMM_Additive
    alphaBoolBlend {blendOption=Linear}
    alphaScaleBiasClamp {bClampResult=true, bInterpResult=true, interpSpeedDecreasing=7, interpSpeedIncreasing=7}
    boneToModify {boneName=dyn_lower_hair_02_r}
    rotation {roll=10}
   }
   n41 = "Transform (Modify) Bone - Bone: dyn_lower_hair_03_r"(ComponentPose=@n40, Alpha=1, Translation="0,0,0", Rotation="0,0,0", Scale="1,1,1") {
    rotationMode=BMM_Additive
    translationMode=BMM_Additive
    alphaBoolBlend {blendOption=Linear}
    alphaScaleBiasClamp {bClampResult=true, bInterpResult=true, interpSpeedDecreasing=7, interpSpeedIncreasing=7}
    boneToModify {boneName=dyn_lower_hair_03_r}
   }
   n42 = "Transform (Modify) Bone - Bone: dyn_lower_hair_01_r"(ComponentPose=@n41, Alpha=1, Translation="0,0,0", Rotation="0,0,-15", Scale="1,1,1") {
    rotationMode=BMM_Additive
    translationMode=BMM_Additive
    alphaBoolBlend {blendOption=Linear}
    alphaScaleBiasClamp {bClampResult=true, bInterpResult=true, interpSpeedIncreasing=8}
    boneToModify {boneName=dyn_lower_hair_01_r}
    rotation {roll=-15}
   }
   n43 = "Transform (Modify) Bone - Bone: dyn_lower_hair_02_r"(ComponentPose=@n42, Alpha=1, Translation="0,1,1", Rotation="0,0,-25", Scale="1,1,1") {
    rotationMode=BMM_Additive
    translationMode=BMM_Additive
    alphaBoolBlend {blendOption=Linear}
    alphaScaleBiasClamp {bClampResult=true, bInterpResult=true, interpSpeedIncreasing=8}
    boneToModify {boneName=dyn_lower_hair_02_r}
    rotation {roll=-25}
    translation {y=1, z=1}
   }
   n44 = "Transform (Modify) Bone - Bone: mstr_lower_hair_01_l"(ComponentPose=@n43, Alpha=1, Translation="0,0,0", Rotation="0,0,-10", Scale="1,1,1") {
    rotationMode=BMM_Additive
    translationMode=BMM_Additive
    alphaBoolBlend {blendOption=Linear}
    alphaScaleBiasClamp {bClampResult=true, bInterpResult=true, interpSpeedIncreasing=8}
    boneToModify {boneName=mstr_lower_hair_01_l}
    rotation {roll=-10}
   }
   n45 = "Transform (Modify) Bone - Bone: mstr_lower_hair_02_l"(ComponentPose=@n44, Alpha=1, Translation="0,0,0", Rotation="0,0,-10", Scale="1,1,1") {
    rotationMode=BMM_Additive
    translationMode=BMM_Additive
    alphaBoolBlend {blendOption=Linear}
    alphaScaleBiasClamp {bClampResult=true, bInterpResult=true, interpSpeedIncreasing=8}
    boneToModify {boneName=mstr_lower_hair_02_l}
    rotation {roll=-10}
   }
   n46 = "Transform (Modify) Bone - Bone: dyn_mid_hair_01_r"(ComponentPose=@n45, Alpha=1, Translation="0,0,0", Rotation="10,0,10", Scale="1,1,1") {
    rotationMode=BMM_Additive
    translationMode=BMM_Additive
    alphaBoolBlend {blendOption=Linear}
    alphaScaleBiasClamp {bClampResult=true, bInterpResult=true, interpSpeedDecreasing=6, interpSpeedIncreasing=8}
    boneToModify {boneName=dyn_mid_hair_01_r}
    rotation {pitch=10, roll=10}
   }
   n47 = "Transform (Modify) Bone - Bone: dyn_mid_hair_02_r"(ComponentPose=@n46, Alpha=1, Translation="0,0,0", Rotation="0,10,0", Scale="1,1,1") {
    rotationMode=BMM_Additive
    translationMode=BMM_Additive
    alphaBoolBlend {blendOption=Linear}
    alphaScaleBiasClamp {bClampResult=true, bInterpResult=true, interpSpeedDecreasing=6, interpSpeedIncreasing=8}
    boneToModify {boneName=dyn_mid_hair_02_r}
    rotation {yaw=10}
   }
   Output Pose(Result=Component To Local(ComponentPose=@n47)) {layerGroup=None}
  ]}
 EventGraph {nodes [
   Event BlueprintUpdateAnimation()
   > n1 = Sequence!(execute) -> (then_0, then_1, then_2:Exec, then_3:Exec, then_4:Exec, then_5:Exec, then_6:Exec, then_7:Exec, then_8:Exec)
   n2 = Cast To FortPlayerAnimInstance(execute=@n1.then_2, Object=GetAnimInstance(self=GetOwnerSkeletalMeshComponent()))
   > n3 = Switch on EFortWeaponCoreAnimation!(execute, Selection:EFortWeaponCoreAnimation Enum=Get WeaponCoreAnim(self=@n2)) -> (
    Melee:Exec
    Pistol:Exec
    Shotgun:Exec
    PaperBlueprint:Exec
    Rifle:Exec
    MeleeOneHand:Exec
    MachinePistol:Exec
    RocketLauncher:Exec
    GrenadeLauncher:Exec
    GoingCommando:Exec
    AssaultRifle:Exec
    TacticalShotgun:Exec
    SniperRifle:Exec
    TrapPlacement:Exec
    ShoulderLauncher:Exec
    AbilityDecoTool:Exec
    Crossbow:Exec
    C4:Exec
    RemoteControl:Exec
    DualWield:Exec
    AR_BullPup:Exec
    AR_ForwardGrip:Exec
    MedPackPaddles:Exec
    SMG_P90:Exec
    AR_DrumGun:Exec
    Consumable_Small:Exec
    Consumable_Large:Exec
    Balloon:Exec
    MountedTurret:Exec
    CreativeTool:Exec
    ExplosiveBow:Exec
    AshtonIndigo:Exec
    AshtonChicago:Exec
    MeleeDualWield:Exec
    Unarmed:Exec
   )
   n4 = Branch(execute=[@n3.AssaultRifle, @n3.Rifle, @n3.Shotgun], Condition=Get bIsTargeting())
   > Set Riffle Equipped(Riffle Equipped:bool=true) -> (Output_Get:bool)
   Set Riffle Equipped(execute=@n4.else, Riffle Equipped=false)
   n5 = Branch(execute=@n3.Pistol, Condition=Get bIsTargeting())
   > Set Pistol Equipped(Pistol Equipped:bool=true) -> (Output_Get:bool)
   Set Pistol Equipped(execute=@n5.else, Pistol Equipped=false)
   n6 = Gravity Override State Function DEPRECATED(
    execute=@n1.then_3
    CharacterPartAnimInstance=Self-Reference()
    GameTimeInSeconds=GetGameTimeInSeconds()
    GravityOverrideParams=Get Gravity Override Params()
   )
   > Set GravityOverride(GravityOverride=@n6)
   Set Targeting(execute=@n1.then_1, Targeting:double="To Float (Boolean)"(InBool=Get Pistol Equipped())) -> (Output_Get:double)
   n7 = "To Float (Boolean)"(InBool=Get Pistol Equipped())
   n8 = float * float(
    A=GetAlphaFromReferenceTransformDelta(SkeletalMeshComponent=Get Skel Mesh(), BoneName=spine_05, BaseName=head, InputRange=(X=0,Y=90), TwistAxis="0,0,1")
    B=3.3
   ) {is=title K2Node_PromotableOperator}
   Set AimingUp(execute=@n1.then_0, AimingUp:double="Clamp (Float)"(Value=float * float(A=@n8, B=@n7), Min=0, Max=1)) -> (Output_Get:double)
   n9 = "To Float (Boolean)"(InBool=Get Riffle Equipped() -> (Riffle Equipped:bool))
   n10 = float * float(
    A=GetAlphaFromReferenceTransformDelta(SkeletalMeshComponent=Get Skel Mesh(), BoneName=spine_05, BaseName=head, InputRange=(X=0,Y=90), TwistAxis="0,0,1")
    B=3.3
   )
   Set RiffleAimingUp(execute=@n1.then_4, RiffleAimingUp:double="Clamp (Float)"(Value=float * float(A=@n10, B=@n9), Min=0, Max=1)) -> (Output_Get:double)
   n11 = Get bIsCrouching()
   n12 = Get bIsMoving2D()
   n13 = AND Boolean(A=NOT Boolean(A=Get bIsTargeting()), B=NOT Boolean(A=AND Boolean(A=@n11, B=@n12)))
   n14 = float * float(
    A=GetAlphaFromReferenceTransformDelta(SkeletalMeshComponent=Get Skel Mesh(), BoneName=spine_05, BaseName=head, InputRange=(X=0,Y=90), TwistAxis="0,0,1")
    B=1
   )
   Set Head_up(execute=@n1.then_5, Head_up:double="Clamp (Float)"(Value=float * float(A=@n14, B:bool=@n13), Min=0, Max=1)) -> (Output_Get:double)
   Event BlueprintInitializeAnimation()
   > Set Skel Mesh(Skel Mesh=GetOwnerSkeletalMeshComponent())
  ]}
]
variables [
 Aiming Down {type=number}
 AimingUp {type=number}
 Body {type=SkeletalMeshComponentBudgeted}
 Gravity Override {type=Vector}
 Gravity Override Params {type=GravityOverrideParamsStruct}
 GravityOverride {type=Vector}
 Head Down {type=number}
 Head_up {type=number}
 Pistol Equipped {type=boolean}
 Riffle Equipped {type=boolean}
 RiffleAimingUp {type=number}
 Skel Mesh {type=SkeletalMeshComponent}
 Targeting {type=number}
 Twistin place {type=number}
]
